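# Animation Blueprint for the RoseBruin biped cosmetic companion, covering locomotion, mood idles, antics and emote reactions.
# A Locomotion state machine uses cached idle and linked-layer run poses, with proxy-table Choosers selecting stops, antics and reactions.
# Mood, face, breathing and turn additives plus a montage slot feed a curve-gated look-at aim offset, Dead Blending and audio-driven jaw curve.
# Slope warping and leg IK nodes exist but are unwired.
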AnimBlueprint /CosmeticCompanions/Assets/Biped/RoseBruin/Blueprint/RoseBruin_AnimBP
target_skeleton=/CosmeticCompanions/Assets/Biped/RoseBruin/Meshes/RoseBruin_Biped_Skeleton
graphs [
 AnimGraph {nodes [
   n1 = Biped Slope Warping!(ComponentPose:Component Space Pose Link Structure=Local To Component(), Alpha:float=1, MaxStepHeight:float=50) -> (Pose:Component Space Pose Link Structure) {
    is=title FortAnimGraphNode_SlopeWarping
    feetDefinitions [
     {numBonesInLimb=2, fKFootBone {boneName=ball_r}, iKFootBone {boneName=IK_Foot_R}}
     {numBonesInLimb=2, fKFootBone {boneName=ball_l}, iKFootBone {boneName=IK_Foot_L}}
    ]
    iKFootRootBone {boneName=IK_Foot_Root}
    pelvisBone {boneName=pelvis}
   }
   n2 = Leg IK!(ComponentPose:Component Space Pose Link Structure=@n1, Alpha:float=1) -> (Pose:Component Space Pose Link Structure) {
    is=title AnimGraphNode_LegIK
    alphaBoolBlend {blendOption=Linear}
    alphaScaleBiasClamp {bias=1, scale=-1}
    legsDefinition [
     {
      bEnableKneeTwistCorrection=true
      bEnableRotationLimit=false
      footBoneForwardAxis=Y
      hingeRotationAxis=None
      minRotationAngle=15
      numBonesInLimb=3
      twistOffsetCurveName=None
      fKFootBone {boneName=ball_r}
      iKFootBone {boneName=IK_Foot_R}
     }
     {
      bEnableKneeTwistCorrection=true
      bEnableRotationLimit=false
      footBoneForwardAxis=Y
      hingeRotationAxis=None
      minRotationAngle=15
      numBonesInLimb=3
      twistOffsetCurveName=None
      fKFootBone {boneName=ball_l}
      iKFootBone {boneName=IK_Foot_L}
     }
    ]
   }
   Component To Local(ComponentPose=@n2)
   n3 = Get CompanionDirection() -> (CompanionDirection:float)
   AnimGraphNode_StateMachine_2 = BodyMood SM() {is=title AnimGraphNode_StateMachine, bAllowConduitEntryStates=true}
   AnimGraphNode_StateMachine_3 = Locomotion SM() {is=title AnimGraphNode_StateMachine}
   AnimGraphNode_StateMachine_1 = FaceAntic SM() {is=title AnimGraphNode_StateMachine}
   n4 = Apply Additive(
    Base=Apply Additive(
     Base=Apply Additive(Base=@AnimGraphNode_StateMachine_3, Additive=@AnimGraphNode_StateMachine_2) {alphaBoolBlend {blendOption=Linear}}
     Additive=@AnimGraphNode_StateMachine_1
    ) {alphaBoolBlend {blendOption=Linear}}
    Additive="Sequence Player (additive) 'RoseBruin_COR_Breathing_ADD'"() {sequence=/CosmeticCompanions/Assets/Biped/RoseBruin/Animation/Poses/RoseBruin_COR_Breathing_ADD}
   ) {alphaBoolBlend {blendOption=Linear}}
   n5 = AimOffset Player 'RoseBruin_AO'!(
    X:float=Get LookAtYawAdjusted() -> (LookAtYawAdjusted:double)
    Y:float=Get LookAtPitchAdjusted() -> (LookAtPitchAdjusted:double)
    BasePose:Pose Link Structure=Apply Additive(
     Base=Slot 'DefaultSlot'(Source=@n4)
     Additive=Blendspace Player 'RoseBruin_COR_Loco_Turn_BS'!(X=@n3) -> (Pose) {blendSpace=/CosmeticCompanions/Assets/Biped/RoseBruin/Animation/Core/RoseBruin_COR_Loco_Turn_BS}
    ) {alphaBoolBlend {blendOption=Linear}}
    AlphaCurveName:Name=DisableAimOffset
   ) -> (Pose:Pose Link Structure) {
    alphaInputType=Curve
    blendSpace=/CosmeticCompanions/Assets/Biped/RoseBruin/Animation/AimOffsets/RoseBruin_AO
    alphaBoolBlend {blendOption=Linear}
    alphaScaleBiasClamp {bias=1, scale=-1}
   }
   Output Pose(Result=Modify Curve(
     SourcePose=Dead Blending!(Source:Pose Link Structure=@n5) -> (Pose:Pose Link Structure) {is=title AnimGraphNode_DeadBlending, defaultBlendMode=Linear}
     CurveValues_0=1
     Alpha=Get AudioEnvelopeValue() -> (AudioEnvelopeValue:float)
    ) {curveNames [jaw_open_pose], curveValues [0]})
   AnimGraphNode_StateMachine_5 = Idle SM() {is=title AnimGraphNode_StateMachine, bAllowConduitEntryStates=true}
   Save cached pose 'IdlePose'(Pose=@AnimGraphNode_StateMachine_5) {cacheName=IdlePose}
   Save cached pose 'RunPose'(Pose=LocomotionRun!() -> (Pose:Pose Link Structure) {is=title AnimGraphNode_LinkedAnimLayer, layer=LocomotionRun}) {cacheName=RunPose}
  ]}
 FaceAntic SM {
  entry=Default
  path=AnimGraph.AnimGraphNode_StateMachine_1.FaceAntic SM
  ids=id
  nodes [
   AnimStateNode_0 = Default() {is=title AnimStateNode}
   AnimStateNode_1 = Antic() {is=title AnimStateNode}
   AnimStateTransitionNode_0 = AnimStateTransitionNode() {crossfadeDuration=0, logicType=TLT_Inertialization}
   AnimStateTransitionNode_1 = AnimStateTransitionNode() {automaticRuleTriggerTime=0, bAutomaticRuleBasedOnSequencePlayerInState=true, crossfadeDuration=0, logicType=TLT_Inertialization}
   AnimStateTransitionNode_2 = AnimStateTransitionNode()
  ]
  links [
   {
    dst=Default
    rule=AnimGraph.AnimGraphNode_StateMachine_1.FaceAntic SM.AnimStateTransitionNode_1.Transition
    src=Antic
    via=AnimStateTransitionNode_1
   }
   {
    dst=Default
    rule=AnimGraph.AnimGraphNode_StateMachine_1.FaceAntic SM.AnimStateTransitionNode_2.Transition
    src=Antic
    via=AnimStateTransitionNode_2
   }
   {
    dst=Antic
    rule=AnimGraph.AnimGraphNode_StateMachine_1.FaceAntic SM.AnimStateTransitionNode_0.Transition
    src=Default
    via=AnimStateTransitionNode_0
   }
  ]
 }
 Default {
  path=AnimGraph.AnimGraphNode_StateMachine_1.FaceAntic SM.AnimStateNode_0.Default
  nodes [Output Animation Pose(Result=Additive Identity Pose())]
 }
 Antic {
  path=AnimGraph.AnimGraphNode_StateMachine_1.FaceAntic SM.AnimStateNode_1.Antic
  nodes [
   AnimGraphNode_ChooserPlayer_0 = Chooser Player: PRA_Companion_FacePose!() -> (Pose:Pose Link Structure) {
    is=title AnimGraphNode_ChooserPlayer
    chooser {
     _structType=/Script/ProxyTable.LookupProxy
     proxy=/CosmeticCompanions/Chooser/PRA_Companion_FacePose
     proxyTable {
      _structType=/Script/ProxyTable.ProxyTableContextProperty
      binding {contextIndex=0, displayName=Chooser Tables, isBoundToRoot=false, propertyBindingChain [ChooserTables]}
     }
    }
    chooserContextDefinition [
     {
      _structType=/Script/Chooser.ContextObjectTypeClass
      class=&/RoseBruin_AnimBP.RoseBruin_AnimBP_C
      direction=ReadWrite
     }
     {struct=/Script/Chooser.ChooserPlayerSettings, _structType=/Script/Chooser.ContextObjectTypeStruct, direction=Read}
    ]
    defaultSettings {blendOption=Linear}
   }
   Output Animation Pose(Result=@AnimGraphNode_ChooserPlayer_0)
  ]
 }
 AnimationBlendStackGraph_0 {
  path=AnimGraph.AnimGraphNode_StateMachine_1.FaceAntic SM.AnimStateNode_1.Antic.AnimGraphNode_ChooserPlayer_0.AnimationBlendStackGraph_0
  nodes [Output Pose!(Result:Pose Link Structure=Blend Stack Input!() -> (Pose:Pose Link Structure) {is=title AnimGraphNode_BlendStackInput}) {is=title AnimGraphNode_BlendStackResult}]
 }
 Transition {
  path=AnimGraph.AnimGraphNode_StateMachine_1.FaceAntic SM.AnimStateTransitionNode_0.Transition
  nodes [Result(bCanEnterTransition=Get bTriggerFaceAntic() -> (bTriggerFaceAntic:bool))]
 }
 Transition {path=AnimGraph.AnimGraphNode_StateMachine_1.FaceAntic SM.AnimStateTransitionNode_1.Transition, nodes [Result()]}
 Transition {
  path=AnimGraph.AnimGraphNode_StateMachine_1.FaceAntic SM.AnimStateTransitionNode_2.Transition
  nodes [Result(bCanEnterTransition=Get bDisableFaceAntics() -> (bDisableFaceAntics:bool))]
 }
 BodyMood SM {
  entry=Conduit
  path=AnimGraph.AnimGraphNode_StateMachine_2.BodyMood SM
  ids=id
  nodes [
   AnimStateConduitNode_0 = Conduit()
   AnimStateNode_0 = Default()
   AnimStateNode_1 = Worried() {is=title AnimStateNode}
   AnimStateNode_2 = Combat() {is=title AnimStateNode}
   AnimStateTransitionNode_0 = AnimStateTransitionNode() {bAllowInertializationForSelfTransitions=true}
   AnimStateTransitionNode_11 = AnimStateTransitionNode() {bAllowInertializationForSelfTransitions=true}
   AnimStateTransitionNode_12 = AnimStateTransitionNode() {bAllowInertializationForSelfTransitions=true}
   AnimStateTransitionNode_15 = AnimStateTransitionNode()
   AnimStateTransitionNode_16 = AnimStateTransitionNode() {bAllowInertializationForSelfTransitions=true}
   AnimStateTransitionNode_7 = AnimStateTransitionNode() {bAllowInertializationForSelfTransitions=true, priorityOrder=2}
   AnimStateTransitionNode_8 = AnimStateTransitionNode() {bAllowInertializationForSelfTransitions=true}
  ]
  links [
   {
    dst=Conduit
    rule=AnimGraph.AnimGraphNode_StateMachine_2.BodyMood SM.AnimStateTransitionNode_12.Transition
    src=Combat
    via=AnimStateTransitionNode_12
   }
   {
    dst=Combat
    rule=AnimGraph.AnimGraphNode_StateMachine_2.BodyMood SM.AnimStateTransitionNode_0.AnimationTransitionGraph_2
    src=Conduit
    via=AnimStateTransitionNode_0
   }
   {
    dst=Default
    rule=AnimGraph.AnimGraphNode_StateMachine_2.BodyMood SM.AnimStateTransitionNode_7.Transition
    src=Conduit
    via=AnimStateTransitionNode_7
   }
   {
    dst=Worried
    rule=AnimGraph.AnimGraphNode_StateMachine_2.BodyMood SM.AnimStateTransitionNode_8.Transition
    src=Conduit
    via=AnimStateTransitionNode_8
   }
   {
    dst=Combat
    rule=AnimGraph.AnimGraphNode_StateMachine_2.BodyMood SM.AnimStateTransitionNode_16.Transition
    src=Default
    via=AnimStateTransitionNode_16
   }
   {
    dst=Worried
    rule=AnimGraph.AnimGraphNode_StateMachine_2.BodyMood SM.AnimStateTransitionNode_15.Transition
    src=Default
    via=AnimStateTransitionNode_15
   }
   {
    dst=Conduit
    rule=AnimGraph.AnimGraphNode_StateMachine_2.BodyMood SM.AnimStateTransitionNode_11.Transition
    src=Worried
    via=AnimStateTransitionNode_11
   }
  ]
 }
 Conduit {path=AnimGraph.AnimGraphNode_StateMachine_2.BodyMood SM.AnimStateConduitNode_0.Conduit, nodes [Result(bCanEnterTransition=true)]}
 Default {
  path=AnimGraph.AnimGraphNode_StateMachine_2.BodyMood SM.AnimStateNode_0.Default
  nodes [Output Animation Pose(Result=Additive Identity Pose())]
 }
 Worried {
  path=AnimGraph.AnimGraphNode_StateMachine_2.BodyMood SM.AnimStateNode_1.Worried
  nodes [Output Animation Pose(Result=Two Way Blend(
     A="Sequence Player (additive) 'RoseBruin_OLD_Pose_Worried'"() {bLoopAnimation=false, sequence=/CosmeticCompanions/Assets/Biped/RoseBruin/Animation/Developer/RoseBruin_OLD_Pose_Worried}
     B="Sequence Player (additive) 'RoseBruin_OLD_Pose_Worried_Locomotion'"() {bLoopAnimation=false, sequence=/CosmeticCompanions/Assets/Biped/RoseBruin/Animation/Developer/RoseBruin_OLD_Pose_Worried_Locomotion}
     Alpha=Get LocomotionAlpha() -> (LocomotionAlpha:double)
    ))]
 }
 Combat {
  path=AnimGraph.AnimGraphNode_StateMachine_2.BodyMood SM.AnimStateNode_2.Combat
  nodes [Output Animation Pose(Result=Two Way Blend(
     A=Additive Identity Pose()
     B="Sequence Player (additive) 'RoseBruin_OLD_Pose_Angry_Locomotion'"() {bLoopAnimation=false, sequence=/CosmeticCompanions/Assets/Biped/RoseBruin/Animation/Developer/RoseBruin_OLD_Pose_Angry_Locomotion}
     Alpha=Get LocomotionAlpha()
    ))]
 }
 AnimationTransitionGraph_2 {
  path=AnimGraph.AnimGraphNode_StateMachine_2.BodyMood SM.AnimStateTransitionNode_0.AnimationTransitionGraph_2
  nodes [Result(bCanEnterTransition=Get bIsMoodCombat() -> (bIsMoodCombat:bool))]
 }
 Transition {
  path=AnimGraph.AnimGraphNode_StateMachine_2.BodyMood SM.AnimStateTransitionNode_11.Transition
  nodes [Result(bCanEnterTransition=NOT Boolean(A=Get bIsMoodWorried() -> (bIsMoodWorried:bool)))]
 }
 Transition {
  path=AnimGraph.AnimGraphNode_StateMachine_2.BodyMood SM.AnimStateTransitionNode_12.Transition
  nodes [Result(bCanEnterTransition=NOT Boolean(A=Get bIsMoodCombat()))]
 }
 Transition {
  path=AnimGraph.AnimGraphNode_StateMachine_2.BodyMood SM.AnimStateTransitionNode_15.Transition
  nodes [Result(bCanEnterTransition=Get bIsMoodWorried())]
 }
 Transition {
  path=AnimGraph.AnimGraphNode_StateMachine_2.BodyMood SM.AnimStateTransitionNode_16.Transition
  nodes [Result(bCanEnterTransition=Get bIsMoodCombat())]
 }
 Transition {path=AnimGraph.AnimGraphNode_StateMachine_2.BodyMood SM.AnimStateTransitionNode_7.Transition, nodes [Result(bCanEnterTransition=true)]}
 Transition {
  path=AnimGraph.AnimGraphNode_StateMachine_2.BodyMood SM.AnimStateTransitionNode_8.Transition
  nodes [Result(bCanEnterTransition=Get bIsMoodWorried())]
 }
 Locomotion SM {
  entry=Idle
  path=AnimGraph.AnimGraphNode_StateMachine_3.Locomotion SM
  ids=id
  nodes [
   AnimStateNode_4 = Walk() {
    is=title AnimStateNode
    stateEntered {guid=#1}
    stateFullyBlended {guid=#2}
    stateLeft {guid=#3}
   }
   AnimStateNode_5 = Idle() {
    stateEntered {guid=#4}
    stateFullyBlended {guid=#5}
    stateLeft {guid=#6}
   }
   AnimStateNode_6 = Stop() {
    is=title AnimStateNode
    stateEntered {guid=#7}
    stateFullyBlended {guid=#8}
    stateLeft {guid=#9}
   }
   AnimStateTransitionNode_10 = AnimStateTransitionNode() {automaticRuleTriggerTime=0, bAutomaticRuleBasedOnSequencePlayerInState=true}
   AnimStateTransitionNode_11 = AnimStateTransitionNode() {crossfadeDuration=0.15}
   AnimStateTransitionNode_12 = AnimStateTransitionNode() {blendMode=CircularInOut, crossfadeDuration=0.3}
   AnimStateTransitionNode_6 = AnimStateTransitionNode() {crossfadeDuration=0.1, logicType=TLT_Inertialization}
   AnimStateTransitionNode_9 = AnimStateTransitionNode()
  ]
  links [
   {
    dst=Walk
    rule=AnimGraph.AnimGraphNode_StateMachine_3.Locomotion SM.AnimStateTransitionNode_6.AnimationTransitionGraph_9
    src=Idle
    via=AnimStateTransitionNode_6
   }
   {
    dst=Idle
    rule=AnimGraph.AnimGraphNode_StateMachine_3.Locomotion SM.AnimStateTransitionNode_10.AnimationTransitionGraph_9
    src=Stop
    via=AnimStateTransitionNode_10
   }
   {
    dst=Idle
    rule=AnimGraph.AnimGraphNode_StateMachine_3.Locomotion SM.AnimStateTransitionNode_12.AnimationTransitionGraph_9
    src=Stop
    via=AnimStateTransitionNode_12
   }
   {
    dst=Walk
    rule=AnimGraph.AnimGraphNode_StateMachine_3.Locomotion SM.AnimStateTransitionNode_11.AnimationTransitionGraph_9
    src=Stop
    via=AnimStateTransitionNode_11
   }
   {
    dst=Stop
    rule=AnimGraph.AnimGraphNode_StateMachine_3.Locomotion SM.AnimStateTransitionNode_9.AnimationTransitionGraph_9
    src=Walk
    via=AnimStateTransitionNode_9
   }
  ]
 }
 Walk {
  path=AnimGraph.AnimGraphNode_StateMachine_3.Locomotion SM.AnimStateNode_4.Walk
  nodes [Output Animation Pose(Result=Use cached pose 'RunPose'())]
 }
 Idle {
  path=AnimGraph.AnimGraphNode_StateMachine_3.Locomotion SM.AnimStateNode_5.Idle
  nodes [
   AnimGraphNode_StateMachine_1 = AnticsSM() {is=title AnimGraphNode_StateMachine, bAllowConduitEntryStates=true}
   AnimGraphNode_StateMachine_2 = AnticPoseSM() {is=title AnimGraphNode_StateMachine}
   n1 = Blend Poses by bool(
    BlendPose_0=Use cached pose 'IdlePose'()
    BlendPose_1=Apply Additive(
     Base=Use cached pose 'IdlePose'()
     Additive=Apply Additive!(Base=@AnimGraphNode_StateMachine_1, Additive=@AnimGraphNode_StateMachine_2, bAlphaBoolEnabled:bool=True) -> (Pose) {alphaInputType=Bool, alphaBoolBlend {blendOption=Linear}}
     Alpha=Get AnticsAlpha() -> (AnticsAlpha:double)
    ) {alphaBoolBlend {blendOption=Linear}}
    BlendTime_0=0.5
    BlendTime_1=0.5
    bActiveValue=Get bUseLowFidelity() -> (bUseLowFidelity:bool)
   ) {
    blendProfile=/CosmeticCompanions/Assets/Quadruped/SitPlant/Meshes/SitPlant_Quadruped_Skeleton:TimeBlend_FastFeetSlowBody
    blendType=Linear
    transitionType=Inertialization
   }
   Output Animation Pose(Result=Modify Curve(SourcePose=@n1, CurveValues_0=1, Alpha=1) {curveNames [idle], curveValues [0]})
  ]
 }
 AnticsSM {
  entry=Conduit
  path=AnimGraph.AnimGraphNode_StateMachine_3.Locomotion SM.AnimStateNode_5.Idle.AnimGraphNode_StateMachine_1.AnticsSM
  ids=id
  nodes [
   AnimStateConduitNode_0 = Conduit()
   AnimStateNode_0 = DefaultState() {is=title AnimStateNode}
   AnimStateNode_1 = HeadAnticChooser() {is=title AnimStateNode}
   AnimStateTransitionNode_0 = AnimStateTransitionNode() {logicType=TLT_Inertialization, priorityOrder=100}
   AnimStateTransitionNode_1 = AnimStateTransitionNode() {logicType=TLT_Inertialization}
   AnimStateTransitionNode_2 = AnimStateTransitionNode() {bAutomaticRuleBasedOnSequencePlayerInState=true}
   AnimStateTransitionNode_5 = AnimStateTransitionNode()
  ]
  links [
   {
    dst=DefaultState
    rule=AnimGraph.AnimGraphNode_StateMachine_3.Locomotion SM.AnimStateNode_5.Idle.AnimGraphNode_StateMachine_1.AnticsSM.AnimStateTransitionNode_0.AnimationTransitionGraph_4
    src=Conduit
    via=AnimStateTransitionNode_0
   }
   {
    dst=HeadAnticChooser
    rule=AnimGraph.AnimGraphNode_StateMachine_3.Locomotion SM.AnimStateNode_5.Idle.AnimGraphNode_StateMachine_1.AnticsSM.AnimStateTransitionNode_1.AnimationTransitionGraph_4
    src=Conduit
    via=AnimStateTransitionNode_1
   }
   {
    dst=Conduit
    rule=AnimGraph.AnimGraphNode_StateMachine_3.Locomotion SM.AnimStateNode_5.Idle.AnimGraphNode_StateMachine_1.AnticsSM.AnimStateTransitionNode_5.AnimationTransitionGraph_4
    src=DefaultState
    via=AnimStateTransitionNode_5
   }
   {
    dst=Conduit
    rule=AnimGraph.AnimGraphNode_StateMachine_3.Locomotion SM.AnimStateNode_5.Idle.AnimGraphNode_StateMachine_1.AnticsSM.AnimStateTransitionNode_2.AnimationTransitionGraph_4
    src=HeadAnticChooser
    via=AnimStateTransitionNode_2
   }
  ]
 }
 Conduit {
  path=AnimGraph.AnimGraphNode_StateMachine_3.Locomotion SM.AnimStateNode_5.Idle.AnimGraphNode_StateMachine_1.AnticsSM.AnimStateConduitNode_0.Conduit
  nodes [Result(bCanEnterTransition=true)]
 }
 DefaultState {
  path=AnimGraph.AnimGraphNode_StateMachine_3.Locomotion SM.AnimStateNode_5.Idle.AnimGraphNode_StateMachine_1.AnticsSM.AnimStateNode_0.DefaultState
  nodes [Output Animation Pose(Result=Additive Identity Pose())]
 }
 HeadAnticChooser {
  path=AnimGraph.AnimGraphNode_StateMachine_3.Locomotion SM.AnimStateNode_5.Idle.AnimGraphNode_StateMachine_1.AnticsSM.AnimStateNode_1.HeadAnticChooser
  nodes [
   AnimGraphNode_ChooserPlayer_0 = Chooser Player: PRA_Companion_AnticsSelection!() -> (Pose:Pose Link Structure) {
    is=title AnimGraphNode_ChooserPlayer
    chooser {
     _structType=&LookupProxy
     proxy=/CosmeticCompanions/Chooser/PRA_Companion_AnticsSelection
     proxyTable {
      _structType=&ProxyTableContextProperty
      binding {contextIndex=0, displayName=Chooser Tables, isBoundToRoot=false, propertyBindingChain [ChooserTables]}
     }
    }
    chooserContextDefinition [
     {
      _structType=&ContextObjectTypeClass
      class=&RoseBruin_AnimBP_C
      direction=ReadWrite
     }
     {struct=&ChooserPlayerSettings, _structType=&ContextObjectTypeStruct, direction=Read}
    ]
    defaultSettings {blendOption=Linear}
   }
   Output Animation Pose(Result=Modify Curve(SourcePose=Sequence Player!(Sequence:Anim Sequence Base Object Reference) -> (Pose) {bLoopAnimation=false}, CurveValues_0=1, Alpha=1) {curveNames [DisableFaceAntics], curveValues [0]})
  ]
 }
 AnimationBlendStackGraph_0 {
  path=AnimGraph.AnimGraphNode_StateMachine_3.Locomotion SM.AnimStateNode_5.Idle.AnimGraphNode_StateMachine_1.AnticsSM.AnimStateNode_1.HeadAnticChooser.AnimGraphNode_ChooserPlayer_0.AnimationBlendStackGraph_0
  nodes [Output Pose(Result=Blend Stack Input())]
 }
 AnimationTransitionGraph_4 {
  path=AnimGraph.AnimGraphNode_StateMachine_3.Locomotion SM.AnimStateNode_5.Idle.AnimGraphNode_StateMachine_1.AnticsSM.AnimStateTransitionNode_0.AnimationTransitionGraph_4
  nodes [Result(bCanEnterTransition=true)]
 }
 AnimationTransitionGraph_4 {
  path=AnimGraph.AnimGraphNode_StateMachine_3.Locomotion SM.AnimStateNode_5.Idle.AnimGraphNode_StateMachine_1.AnticsSM.AnimStateTransitionNode_1.AnimationTransitionGraph_4
  nodes [
   Result(bCanEnterTransition=Get!() -> (AnticSequenceInfo_AnticSequenceToPlay:Anim Sequence Object Reference, AnticSequenceInfo_bAnticSequenceReady:bool=~).AnticSequenceInfo_bAnticSequenceReady)
   Break Companion Random Number Timer!(CompanionRandomNumberTimer:"Companion Random Number Timer Structure (by ref)"=Get AnticsTimer() -> (AnticsTimer:Companion Random Number Timer Structure)) -> (ValueInt:int, ValueFloat:float=~, bJustChanged:bool=~) {is=title K2Node_BreakStruct}
  ]
 }
 AnimationTransitionGraph_4 {
  path=AnimGraph.AnimGraphNode_StateMachine_3.Locomotion SM.AnimStateNode_5.Idle.AnimGraphNode_StateMachine_1.AnticsSM.AnimStateTransitionNode_2.AnimationTransitionGraph_4
  nodes [Result()]
 }
 AnimationTransitionGraph_4 {
  path=AnimGraph.AnimGraphNode_StateMachine_3.Locomotion SM.AnimStateNode_5.Idle.AnimGraphNode_StateMachine_1.AnticsSM.AnimStateTransitionNode_5.AnimationTransitionGraph_4
  nodes [Result(bCanEnterTransition=NOT Boolean(A=Get bIsMoodWorried()))]
 }
 AnticPoseSM {
  entry=DefaultPose
  path=AnimGraph.AnimGraphNode_StateMachine_3.Locomotion SM.AnimStateNode_5.Idle.AnimGraphNode_StateMachine_2.AnticPoseSM
  ids=id
  nodes [
   AnimStateNode_0 = DefaultPose() {is=title AnimStateNode}
   AnimStateNode_1 = AdditivePose() {is=title AnimStateNode}
   AnimStateTransitionNode_0 = AnimStateTransitionNode()
   AnimStateTransitionNode_1 = AnimStateTransitionNode() {automaticRuleTriggerTime=0, bAutomaticRuleBasedOnSequencePlayerInState=true}
  ]
  links [
   {
    dst=DefaultPose
    rule=AnimGraph.AnimGraphNode_StateMachine_3.Locomotion SM.AnimStateNode_5.Idle.AnimGraphNode_StateMachine_2.AnticPoseSM.AnimStateTransitionNode_1.Transition
    src=AdditivePose
    via=AnimStateTransitionNode_1
   }
   {
    dst=AdditivePose
    rule=AnimGraph.AnimGraphNode_StateMachine_3.Locomotion SM.AnimStateNode_5.Idle.AnimGraphNode_StateMachine_2.AnticPoseSM.AnimStateTransitionNode_0.Transition
    src=DefaultPose
    via=AnimStateTransitionNode_0
   }
  ]
 }
 DefaultPose {
  path=AnimGraph.AnimGraphNode_StateMachine_3.Locomotion SM.AnimStateNode_5.Idle.AnimGraphNode_StateMachine_2.AnticPoseSM.AnimStateNode_0.DefaultPose
  nodes [Output Animation Pose(Result=Additive Identity Pose())]
 }
 AdditivePose {
  path=AnimGraph.AnimGraphNode_StateMachine_3.Locomotion SM.AnimStateNode_5.Idle.AnimGraphNode_StateMachine_2.AnticPoseSM.AnimStateNode_1.AdditivePose
  nodes [
   AnimGraphNode_ChooserPlayer_0 = Chooser Player: PRA_Companion_BodyPose!() -> (Pose:Pose Link Structure) {
    is=title AnimGraphNode_ChooserPlayer
    chooser {
     _structType=&LookupProxy
     proxy=/CosmeticCompanions/Chooser/PRA_Companion_BodyPose
     proxyTable {
      _structType=&ProxyTableContextProperty
      binding {contextIndex=0, displayName=Chooser Tables, isBoundToRoot=false, propertyBindingChain [ChooserTables]}
     }
    }
    chooserContextDefinition [
     {
      _structType=&ContextObjectTypeClass
      class=&RoseBruin_AnimBP_C
      direction=ReadWrite
     }
     {struct=&ChooserPlayerSettings, _structType=&ContextObjectTypeStruct, direction=Read}
    ]
    defaultSettings {blendOption=Linear}
   }
   Output Animation Pose(Result=@AnimGraphNode_ChooserPlayer_0)
  ]
 }
 AnimationBlendStackGraph_0 {
  path=AnimGraph.AnimGraphNode_StateMachine_3.Locomotion SM.AnimStateNode_5.Idle.AnimGraphNode_StateMachine_2.AnticPoseSM.AnimStateNode_1.AdditivePose.AnimGraphNode_ChooserPlayer_0.AnimationBlendStackGraph_0
  nodes [Output Pose(Result=Blend Stack Input())]
 }
 Transition {
  path=AnimGraph.AnimGraphNode_StateMachine_3.Locomotion SM.AnimStateNode_5.Idle.AnimGraphNode_StateMachine_2.AnticPoseSM.AnimStateTransitionNode_0.Transition
  nodes [
   Result(bCanEnterTransition=Break Companion Random Number Timer(CompanionRandomNumberTimer=Get AnticsTimer()) -> (ValueFloat=~, bJustChanged=~).bJustChanged)
   Get AuxiliaryTimer1() -> (AuxiliaryTimer1:Companion Random Number Timer Structure)
  ]
 }
 Transition {
  path=AnimGraph.AnimGraphNode_StateMachine_3.Locomotion SM.AnimStateNode_5.Idle.AnimGraphNode_StateMachine_2.AnticPoseSM.AnimStateTransitionNode_1.Transition
  nodes [Result()]
 }
 Stop {
  path=AnimGraph.AnimGraphNode_StateMachine_3.Locomotion SM.AnimStateNode_6.Stop
  nodes [
   AnimGraphNode_ChooserPlayer_0 = Chooser Player: PRA_Companion_Stops!() -> (Pose:Pose Link Structure) {
    is=title AnimGraphNode_ChooserPlayer
    chooser {
     _structType=&LookupProxy
     proxy=/CosmeticCompanions/Chooser/PRA_Companion_Stops
     proxyTable {
      _structType=&ProxyTableContextProperty
      binding {contextIndex=0, displayName=Chooser Tables, isBoundToRoot=false, propertyBindingChain [ChooserTables]}
     }
    }
    chooserContextDefinition [
     {
      _structType=&ContextObjectTypeClass
      class=&RoseBruin_AnimBP_C
      direction=ReadWrite
     }
     {struct=&ChooserPlayerSettings, _structType=&ContextObjectTypeStruct, direction=Read}
    ]
   }
   Output Animation Pose(Result=@AnimGraphNode_ChooserPlayer_0)
  ]
 }
 AnimationBlendStackGraph_0 {
  path=AnimGraph.AnimGraphNode_StateMachine_3.Locomotion SM.AnimStateNode_6.Stop.AnimGraphNode_ChooserPlayer_0.AnimationBlendStackGraph_0
  nodes [Output Pose(Result=Blend Stack Input())]
 }
 AnimationTransitionGraph_9 {path=AnimGraph.AnimGraphNode_StateMachine_3.Locomotion SM.AnimStateTransitionNode_10.AnimationTransitionGraph_9, nodes [Result()]}
 AnimationTransitionGraph_9 {
  path=AnimGraph.AnimGraphNode_StateMachine_3.Locomotion SM.AnimStateTransitionNode_11.AnimationTransitionGraph_9
  nodes [Result(bCanEnterTransition=Get bTriggerStopToStart() -> (bTriggerStopToStart:bool))]
 }
 AnimationTransitionGraph_9 {
  path=AnimGraph.AnimGraphNode_StateMachine_3.Locomotion SM.AnimStateTransitionNode_12.AnimationTransitionGraph_9
  nodes [Result(bCanEnterTransition=Get bIsMoodCombat())]
 }
 AnimationTransitionGraph_9 {
  path=AnimGraph.AnimGraphNode_StateMachine_3.Locomotion SM.AnimStateTransitionNode_6.AnimationTransitionGraph_9
  nodes [Result(bCanEnterTransition=Get bIsMoving() -> (bIsMoving:bool))]
 }
 AnimationTransitionGraph_9 {
  path=AnimGraph.AnimGraphNode_StateMachine_3.Locomotion SM.AnimStateTransitionNode_9.AnimationTransitionGraph_9
  nodes [Result(bCanEnterTransition=Get bTriggerStop() -> (bTriggerStop:bool))]
 }
 Idle SM {
  entry=Conduit
  path=AnimGraph.AnimGraphNode_StateMachine_5.Idle SM
  ids=id
  nodes [
   CombatAlias() {is=title AnimStateAliasNode, also {aliases [Combat,CombatIntro]}}
   BoredAlias() {is=title AnimStateAliasNode, also {aliases [Bored,BoredIntro]}}
   AnimStateConduitNode_0 = Conduit()
   AnimStateNode_0 = Idle()
   AnimStateNode_1 = ReactionLoop() {is=title AnimStateNode}
   AnimStateNode_2 = ReactionLoopExitTransition() {is=title AnimStateNode}
   AnimStateNode_4 = CombatIntro() {is=title AnimStateNode}
   AnimStateNode_5 = Combat()
   AnimStateNode_6 = Bored() {is=title AnimStateNode}
   AnimStateNode_7 = BoredIntro() {is=title AnimStateNode}
   AnimStateNode_8 = BoredOutro() {is=title AnimStateNode}
   AnimStateNode_9 = CombatOutro() {is=title AnimStateNode}
   AnimStateTransitionNode_0 = AnimStateTransitionNode() {logicType=TLT_Inertialization, priorityOrder=2}
   AnimStateTransitionNode_1 = AnimStateTransitionNode() {logicType=TLT_Inertialization}
   AnimStateTransitionNode_10 = AnimStateTransitionNode() {bAllowInertializationForSelfTransitions=true}
   AnimStateTransitionNode_11 = AnimStateTransitionNode() {automaticRuleTriggerTime=0, bAllowInertializationForSelfTransitions=true, bAutomaticRuleBasedOnSequencePlayerInState=true}
   AnimStateTransitionNode_13 = AnimStateTransitionNode() {automaticRuleTriggerTime=0, bAutomaticRuleBasedOnSequencePlayerInState=true}
   AnimStateTransitionNode_14 = AnimStateTransitionNode()
   AnimStateTransitionNode_15 = AnimStateTransitionNode() {bAllowInertializationForSelfTransitions=true}
   AnimStateTransitionNode_2 = AnimStateTransitionNode() {crossfadeDuration=0.3, logicType=TLT_Inertialization}
   AnimStateTransitionNode_3 = AnimStateTransitionNode() {automaticRuleTriggerTime=0, bAutomaticRuleBasedOnSequencePlayerInState=true, crossfadeDuration=0.3, logicType=TLT_Inertialization}
   AnimStateTransitionNode_4 = AnimStateTransitionNode()
   AnimStateTransitionNode_5 = AnimStateTransitionNode() {logicType=TLT_Inertialization}
   AnimStateTransitionNode_6 = AnimStateTransitionNode() {automaticRuleTriggerTime=0, bAllowInertializationForSelfTransitions=true, bAutomaticRuleBasedOnSequencePlayerInState=true}
   AnimStateTransitionNode_7 = AnimStateTransitionNode()
   AnimStateTransitionNode_8 = AnimStateTransitionNode() {bAutomaticRuleBasedOnSequencePlayerInState=true}
  ]
  links [
   {
    dst=BoredOutro
    rule=AnimGraph.AnimGraphNode_StateMachine_5.Idle SM.AnimStateTransitionNode_14.AnimationTransitionGraph_4
    src=BoredAlias
    via=AnimStateTransitionNode_14
   }
   {
    dst=Bored
    rule=AnimGraph.AnimGraphNode_StateMachine_5.Idle SM.AnimStateTransitionNode_13.AnimationTransitionGraph_4
    src=BoredIntro
    via=AnimStateTransitionNode_13
   }
   {
    dst=Idle
    rule=AnimGraph.AnimGraphNode_StateMachine_5.Idle SM.AnimStateTransitionNode_8.Transition
    src=BoredOutro
    via=AnimStateTransitionNode_8
   }
   {
    dst=CombatOutro
    rule=AnimGraph.AnimGraphNode_StateMachine_5.Idle SM.AnimStateTransitionNode_15.Transition
    src=CombatAlias
    via=AnimStateTransitionNode_15
   }
   {
    dst=Combat
    rule=AnimGraph.AnimGraphNode_StateMachine_5.Idle SM.AnimStateTransitionNode_11.Transition
    src=CombatIntro
    via=AnimStateTransitionNode_11
   }
   {
    dst=Idle
    rule=AnimGraph.AnimGraphNode_StateMachine_5.Idle SM.AnimStateTransitionNode_6.Transition
    src=CombatOutro
    via=AnimStateTransitionNode_6
   }
   {
    dst=Idle
    rule=AnimGraph.AnimGraphNode_StateMachine_5.Idle SM.AnimStateTransitionNode_0.Transition
    src=Conduit
    via=AnimStateTransitionNode_0
   }
   {
    dst=ReactionLoop
    rule=AnimGraph.AnimGraphNode_StateMachine_5.Idle SM.AnimStateTransitionNode_1.Transition
    src=Conduit
    via=AnimStateTransitionNode_1
   }
   {
    dst=ReactionLoopExitTransition
    rule=AnimGraph.AnimGraphNode_StateMachine_5.Idle SM.AnimStateTransitionNode_5.Transition
    src=Conduit
    via=AnimStateTransitionNode_5
   }
   {
    dst=BoredIntro
    rule=AnimGraph.AnimGraphNode_StateMachine_5.Idle SM.AnimStateTransitionNode_7.Transition
    src=Idle
    via=AnimStateTransitionNode_7
   }
   {
    dst=CombatIntro
    rule=AnimGraph.AnimGraphNode_StateMachine_5.Idle SM.AnimStateTransitionNode_10.Transition
    src=Idle
    via=AnimStateTransitionNode_10
   }
   {
    dst=Conduit
    rule=AnimGraph.AnimGraphNode_StateMachine_5.Idle SM.AnimStateTransitionNode_4.Transition
    src=Idle
    via=AnimStateTransitionNode_4
   }
   {
    dst=ReactionLoopExitTransition
    rule=AnimGraph.AnimGraphNode_StateMachine_5.Idle SM.AnimStateTransitionNode_2.Transition
    src=ReactionLoop
    via=AnimStateTransitionNode_2
   }
   {
    dst=Idle
    rule=AnimGraph.AnimGraphNode_StateMachine_5.Idle SM.AnimStateTransitionNode_3.Transition
    src=ReactionLoopExitTransition
    via=AnimStateTransitionNode_3
   }
  ]
 }
 Conduit {path=AnimGraph.AnimGraphNode_StateMachine_5.Idle SM.AnimStateConduitNode_0.Conduit, nodes [Result(bCanEnterTransition=true)]}
 Idle {
  path=AnimGraph.AnimGraphNode_StateMachine_5.Idle SM.AnimStateNode_0.Idle
  nodes [Output Animation Pose(Result=Blend Poses by bool(
     BlendPose_0=Sequence Player 'RoseBruin_COR_Angry_Idle_Lp'() {sequence=/CosmeticCompanions/Assets/Biped/RoseBruin/Animation/Core/RoseBruin_COR_Angry_Idle_Lp}
     BlendPose_1=Sequence Player 'RoseBruin_COR_Stand_Lp'() {groupName=Idle, method=SyncGroup, sequence=/CosmeticCompanions/Assets/Biped/RoseBruin/Animation/Poses/RoseBruin_COR_Stand_Lp}
     BlendTime_0=0.3
     BlendTime_1=0.4
     bActiveValue=Get bIsMoodCombat()
    ) {blendType=Linear})]
 }
 ReactionLoop {
  path=AnimGraph.AnimGraphNode_StateMachine_5.Idle SM.AnimStateNode_1.ReactionLoop
  nodes [
   AnimGraphNode_ChooserPlayer_0 = Chooser Player: PRA_Companion_SpecialIdle!() -> (Pose:Pose Link Structure) {
    is=title AnimGraphNode_ChooserPlayer
    chooser {
     _structType=&LookupProxy
     proxy=/CosmeticCompanions/Chooser/PRA_Companion_SpecialIdle
     proxyTable {
      _structType=&ProxyTableContextProperty
      binding {contextIndex=0, displayName=Chooser Tables, isBoundToRoot=false, propertyBindingChain [ChooserTables]}
     }
    }
    chooserContextDefinition [
     {
      _structType=&ContextObjectTypeClass
      class=&RoseBruin_AnimBP_C
      direction=ReadWrite
     }
     {struct=&ChooserPlayerSettings, _structType=&ContextObjectTypeStruct, direction=Read}
    ]
    defaultSettings {blendOption=Linear}
   }
   Output Animation Pose(Result=@AnimGraphNode_ChooserPlayer_0)
  ]
 }
 AnimationBlendStackGraph_0 {
  path=AnimGraph.AnimGraphNode_StateMachine_5.Idle SM.AnimStateNode_1.ReactionLoop.AnimGraphNode_ChooserPlayer_0.AnimationBlendStackGraph_0
  nodes [Output Pose(Result=Blend Stack Input())]
 }
 ReactionLoopExitTransition {
  path=AnimGraph.AnimGraphNode_StateMachine_5.Idle SM.AnimStateNode_2.ReactionLoopExitTransition
  nodes [
   AnimGraphNode_ChooserPlayer_0 = Chooser Player: PRA_Companion_ReactToIdle!() -> (Pose:Pose Link Structure) {
    is=title AnimGraphNode_ChooserPlayer
    chooser {
     _structType=&LookupProxy
     proxy=/CosmeticCompanions/Chooser/PRA_Companion_ReactToIdle
     proxyTable {
      _structType=&ProxyTableContextProperty
      binding {contextIndex=0, displayName=Chooser Tables, isBoundToRoot=false, propertyBindingChain [ChooserTables]}
     }
    }
    chooserContextDefinition [
     {
      _structType=&ContextObjectTypeClass
      class=&RoseBruin_AnimBP_C
      direction=ReadWrite
     }
     {struct=&ChooserPlayerSettings, _structType=&ContextObjectTypeStruct, direction=Read}
    ]
    defaultSettings {blendOption=Linear}
   }
   Output Animation Pose(Result=@AnimGraphNode_ChooserPlayer_0)
  ]
 }
 AnimationBlendStackGraph_0 {
  path=AnimGraph.AnimGraphNode_StateMachine_5.Idle SM.AnimStateNode_2.ReactionLoopExitTransition.AnimGraphNode_ChooserPlayer_0.AnimationBlendStackGraph_0
  nodes [Output Pose(Result=Blend Stack Input())]
 }
 CombatIntro {
  path=AnimGraph.AnimGraphNode_StateMachine_5.Idle SM.AnimStateNode_4.CombatIntro
  nodes [Output Animation Pose(Result=Sequence Player 'RoseBruin_RCT_Angry'() {bLoopAnimation=false, sequence=/CosmeticCompanions/Assets/Biped/RoseBruin/Animation/Reacts/RoseBruin_RCT_Angry})]
 }
 Combat {
  path=AnimGraph.AnimGraphNode_StateMachine_5.Idle SM.AnimStateNode_5.Combat
  nodes [Output Animation Pose(Result=Sequence Player 'RoseBruin_COR_Angry_Idle_Lp'() {sequence=&RoseBruin_COR_Angry_Idle_Lp})]
 }
 Bored {
  path=AnimGraph.AnimGraphNode_StateMachine_5.Idle SM.AnimStateNode_6.Bored
  nodes [Output Animation Pose(Result=Apply Additive!(
     Base=Sequence Player 'RoseBruin_COR_Stand_Lp'() {sequence=&RoseBruin_COR_Stand_Lp}
     Additive="Sequence Player (additive) 'RoseBruin_OLD_Pose_Bored_Additive'"() {bLoopAnimation=false, sequence=/CosmeticCompanions/Assets/Biped/RoseBruin/Animation/Developer/RoseBruin_OLD_Pose_Bored_Additive}
     Alpha:float
    ) -> (Pose))]
 }
 BoredIntro {
  path=AnimGraph.AnimGraphNode_StateMachine_5.Idle SM.AnimStateNode_7.BoredIntro
  nodes [Output Animation Pose(Result=Sequence Player 'RoseBruin_COR_Bored_In'() {bLoopAnimation=false, sequence=/CosmeticCompanions/Assets/Biped/RoseBruin/Animation/Core/RoseBruin_COR_Bored_In})]
 }
 BoredOutro {
  path=AnimGraph.AnimGraphNode_StateMachine_5.Idle SM.AnimStateNode_8.BoredOutro
  nodes [Output Animation Pose(Result=Sequence Player 'RoseBruin_COR_Bored_Out'() {bLoopAnimation=false, sequence=/CosmeticCompanions/Assets/Biped/RoseBruin/Animation/Core/RoseBruin_COR_Bored_Out})]
 }
 CombatOutro {
  path=AnimGraph.AnimGraphNode_StateMachine_5.Idle SM.AnimStateNode_9.CombatOutro
  nodes [Output Animation Pose(Result=Sequence Player 'RoseBruin_COR_Angry_Idle_Out'() {bLoopAnimation=false, sequence=/CosmeticCompanions/Assets/Biped/RoseBruin/Animation/Core/RoseBruin_COR_Angry_Idle_Out})]
 }
 Transition {path=AnimGraph.AnimGraphNode_StateMachine_5.Idle SM.AnimStateTransitionNode_0.Transition, nodes [Result(bCanEnterTransition=true)]}
 Transition {
  path=AnimGraph.AnimGraphNode_StateMachine_5.Idle SM.AnimStateTransitionNode_1.Transition
  nodes [Result(bCanEnterTransition=Get bIsEmoteActive() -> (bIsEmoteActive:bool))]
 }
 Transition {
  path=AnimGraph.AnimGraphNode_StateMachine_5.Idle SM.AnimStateTransitionNode_10.Transition
  nodes [
   Result(bCanEnterTransition=Get bIsMoodCombat())
   Get bIsMoodCombat()
  ]
 }
 Transition {path=AnimGraph.AnimGraphNode_StateMachine_5.Idle SM.AnimStateTransitionNode_11.Transition, nodes [Result()]}
 AnimationTransitionGraph_4 {path=AnimGraph.AnimGraphNode_StateMachine_5.Idle SM.AnimStateTransitionNode_13.AnimationTransitionGraph_4, nodes [Result()]}
 AnimationTransitionGraph_4 {
  path=AnimGraph.AnimGraphNode_StateMachine_5.Idle SM.AnimStateTransitionNode_14.AnimationTransitionGraph_4
  nodes [Result(bCanEnterTransition=NOT Boolean(A=Get bIsMoodBored() -> (bIsMoodBored:bool)))]
 }
 Transition {
  path=AnimGraph.AnimGraphNode_StateMachine_5.Idle SM.AnimStateTransitionNode_15.Transition
  nodes [Result(bCanEnterTransition=NOT Boolean(A=Get bIsMoodCombat()))]
 }
 Transition {
  path=AnimGraph.AnimGraphNode_StateMachine_5.Idle SM.AnimStateTransitionNode_2.Transition
  nodes [
   n1 = Get bIsPlayingEmote()
   Result(bCanEnterTransition=NOT Boolean(A=@n1))
   OR Boolean(A=@n1, B=Get bIsMoodCombat())
  ]
 }
 Transition {path=AnimGraph.AnimGraphNode_StateMachine_5.Idle SM.AnimStateTransitionNode_3.Transition, nodes [Result()]}
 Transition {path=AnimGraph.AnimGraphNode_StateMachine_5.Idle SM.AnimStateTransitionNode_4.Transition, nodes [Result(bCanEnterTransition=true)]}
 Transition {
  path=AnimGraph.AnimGraphNode_StateMachine_5.Idle SM.AnimStateTransitionNode_5.Transition
  nodes [Result(bCanEnterTransition=Get bIsReactionActive() -> (bIsReactionActive:bool))]
 }
 Transition {path=AnimGraph.AnimGraphNode_StateMachine_5.Idle SM.AnimStateTransitionNode_6.Transition, nodes [Result()]}
 Transition {
  path=AnimGraph.AnimGraphNode_StateMachine_5.Idle SM.AnimStateTransitionNode_7.Transition
  nodes [Result(bCanEnterTransition=Get bIsMoodBored())]
 }
 Transition {path=AnimGraph.AnimGraphNode_StateMachine_5.Idle SM.AnimStateTransitionNode_8.Transition, nodes [Result()]}
 EventGraph {nodes [
   TryGetPawnOwner()
   Event BlueprintUpdateAnimation()
  ]}
 LocomotionRun {nodes [
   n1 = Sequence Player 'RoseBruin_COR_Loco_Run'() {groupName=Locomotion, method=SyncGroup, sequence=/CosmeticCompanions/Assets/Biped/RoseBruin/Animation/Locomotion/RoseBruin_COR_Loco_Run}
   n2 = Sequence Player 'RoseBruin_COR_Loco_Trot'() {groupName=Locomotion, method=SyncGroup, sequence=/CosmeticCompanions/Assets/Biped/RoseBruin/Animation/Locomotion/RoseBruin_COR_Loco_Trot}
   n3 = Sequence Player 'RoseBruin_COR_Loco_Sprint'() {groupName=Locomotion, method=SyncGroup, sequence=/CosmeticCompanions/Assets/Biped/RoseBruin/Animation/Locomotion/RoseBruin_COR_Loco_Sprint}
   n4 = Blend Poses by bool(
    BlendPose_0=@n3
    BlendPose_1=Blend Poses by bool(BlendPose_0=@n1, BlendPose_1=@n2, bActiveValue=Get bIsLocoRun() -> (bIsLocoRun:bool)) {blendType=Sinusoidal}
    bActiveValue=Get bIsLocoSprint() -> (bIsLocoSprint:bool)
   ) {blendType=Sinusoidal}
   n5 = Blendspace Player 'RoseBruin_OLD_Lean_BS'!(X=Get CompanionSpeed2D() -> (CompanionSpeed2D:float)) -> (Pose) {blendSpace=/CosmeticCompanions/Assets/Biped/RoseBruin/Animation/Developer/RoseBruin_OLD_Lean_BS}
   Output Pose(Result=Modify Curve(SourcePose=Apply Additive(Base=@n4, Additive=@n5) {alphaBoolBlend {blendOption=Linear}}, CurveValues_0=1, Alpha=1) {curveNames [locomotionalpha], curveValues [0]}) {is=title AnimGraphNode_Root}
  ]}
]
variables [ChooserTables {type=ProxyTable, default=/CosmeticCompanions/Assets/Biped/RoseBruin/Animation/Choosers/PRT_RoseBruin_Choosers}]
